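# Physics asset for the SugarBubble backpack cosmetic, defining collision and joint limits for its dangling charm and strap.
# A four-link charm chain and a two-link strap chain hang from the pack body with tight swing and twist limits, the strap swinging on only one axis.
# Capsules approximate every shape, with three on the charm tip and two on the pack.
# A right upper-arm capsule, angularly free relative to the pack, likely serves as a collider.
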PhysicsAsset /BRCosmetics/Accessories/FORT_Backpacks/Backpack_F_MED_SugarBubble/Meshes/F_MED_SugarBubble_Pack_Physics
bodies {
 charm_1_dyn [{
   shapeType=Capsule
   length=1.2671
   radius=0.5
   shapeName=charm_1_dyn_capsule
   center=<0.1887 0.4077 -0.1013>
   rotation=<rot 0 87.4746 -85.9493>
  }]
 charm_2_dyn [{shapeType=Capsule, length=3.2093, radius=0.5, shapeName=charm_2_dyn_capsule, center=<-0.0004 0.0913 0>, rotation=<rot 0 90.2732 -89.8389>}]
 charm_3_dyn [{shapeType=Capsule, length=2.5221, radius=0.5005, shapeName=charm_3_dyn_capsule, center=<-0.0005 0.0913 0>, rotation=<rot 0 90.3405 -89.4146>}]
 charm_4_dyn [
  {
   shapeType=Capsule
   length=3.3382
   radius=0.8685
   shapeName=charm_4_dyn_capsule
   center=<2.2887 -0.5437 1.5579>
   rotation=<rot 82.1347 0.0001 0>
  }
  {
   shapeType=Capsule
   length=4.6173
   radius=0.769
   shapeName=charm_4_dyn_capsule
   center=<2.3334 -0.4799 0.0061>
   rotation=<rot -89.85 0.0003 -0.0001>
  }
  {
   shapeType=Capsule
   length=3.6504
   radius=0.7901
   shapeName=charm_4_dyn_capsule
   center=<2.3171 -0.4167 -1.4766>
   rotation=<rot 83.4374 -180 -180>
  }
 ]
 pack [
  {shapeType=Capsule, length=29.4084, radius=1.8913, center=<-9.7889 -3.2556 5.7204>, rotation=<rot 3.2942 -86.7679 -40.7302>}
  {shapeType=Capsule, length=44.1124, radius=17.8117, shapeName=pack_capsule, center=<0 19.43 0>, rotation=<rot 48.2791 0 0>}
 ]
 strap_1_dyn [{shapeType=Capsule, length=0.9901, radius=0.505, shapeName=strap_1_dyn_capsule, center=<-0.4909 -0.0035 0>, rotation=<rot 0 0.405 -84.4007>}]
 strap_2_dyn [{shapeType=Capsule, length=1.5489, radius=0.5, shapeName=strap_2_dyn_capsule, rotation=<rot 3.2867 -86.2492 -90>}]
 upperarm_r [{
   shapeType=Capsule
   length=19.7929
   radius=6.1831
   shapeName=upperarm_r_capsule
   center=<-8.1788 -2.2169 -2.3901>
   rotation=<rot -0.0001 -89.9999 88.2598>
  }]
}
constraints [
 {bone1Name=charm_1_dyn, bone2Name=pack, swing1LimitDegrees=17, swing2LimitDegrees=17, twistLimitDegrees=10}
 {bone1Name=charm_2_dyn, bone2Name=charm_1_dyn, swing1LimitDegrees=17, swing2LimitDegrees=19, twistLimitDegrees=10}
 {bone1Name=charm_3_dyn, bone2Name=charm_2_dyn, swing1LimitDegrees=18, swing2LimitDegrees=19, twistLimitDegrees=6}
 {bone1Name=charm_4_dyn, bone2Name=charm_3_dyn, swing1LimitDegrees=16, swing2LimitDegrees=10, twistLimitDegrees=4}
 {
  bone1Name=upperarm_r
  bone2Name=pack
  swing1LimitDegrees=45
  swing1Motion=Free
  swing2LimitDegrees=45
  swing2Motion=Free
  twistLimitDegrees=45
  twistMotion=Free
 }
 {bone1Name=strap_1_dyn, bone2Name=pack, swing1LimitDegrees=0, swing2LimitDegrees=25, twistLimitDegrees=2}
 {bone1Name=strap_2_dyn, bone2Name=strap_1_dyn, swing1LimitDegrees=0, swing2LimitDegrees=15, twistLimitDegrees=2}
]
