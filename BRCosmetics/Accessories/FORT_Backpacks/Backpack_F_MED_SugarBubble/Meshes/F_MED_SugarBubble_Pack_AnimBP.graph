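# Animation Blueprint for the SugarBubble back bling: poses the pack, simulates physics and toggles a material element.
# Additive offsets adjust the pack, with skydiving adding more, including one driven by the owner's spine delta, plus a charm rotation.
# The owner's right upper arm world transform is copied to upperarm_r before the RigidBody node runs.
# A dynamic material parameter hides Element 01 based on the equipped matching pickaxe and movement or emote state.
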AnimBlueprint /BRCosmetics/Accessories/FORT_Backpacks/Backpack_F_MED_SugarBubble/Meshes/F_MED_SugarBubble_Pack_AnimBP
target_skeleton=&/F_MED_SugarBubble_Pack_Skeleton
graphs [
 AnimGraph {nodes [
   n1 = "Transform (Modify) Bone - Bone: pack"(ComponentPose=Local To Component(LocalPose=Input Pose()), Translation="0.000,-0.408,-0.000", Rotation="-0.000,0.000,6.113") {
    rotationMode=BMM_Additive
    scaleMode=BMM_Additive
    translationMode=BMM_Additive
    alphaBoolBlend {blendOption=Linear}
    boneToModify {boneName=pack}
    rotation {roll=6.113}
    translation=<0 -0.4077 0>
   }
   n2 = Get bIsSkyDiving()
   n3 = "Transform (Modify) Bone - Bone: pack"!(ComponentPose=@n1, bAlphaBoolEnabled:bool=@n2, Translation="0.000,-1.649,-0.000", Rotation="-0.000,0.000,3.426", Scale) -> (Pose) {
    alphaInputType=Bool
    rotationMode=BMM_Additive
    scaleMode=BMM_Additive
    translationMode=BMM_Additive
    alphaBoolBlend {blendOption=Linear}
    boneToModify {boneName=pack}
    rotation {roll=3.426}
    translation {y=-1.649}
   }
   n4 = "Transform (Modify) Bone - Bone: pack"!(
    ComponentPose=@n3
    Alpha:float=Get Rotation_RightThigh_z_inSky() -> (Rotation_RightThigh_z_inSky:float)
    Translation="0.000,-3.250,0.000"
    Rotation="-0.000,0.000,20.792"
    Scale
   ) -> (Pose) {
    rotationMode=BMM_Additive
    scaleMode=BMM_Additive
    translationMode=BMM_Additive
    alphaBoolBlend {blendOption=Linear}
    alphaScaleBiasClamp {bInterpResult=true, interpSpeedDecreasing=1, interpSpeedIncreasing=5}
    boneToModify {boneName=pack}
    rotation {roll=20.792}
    translation {y=-3.2498, z=0}
   }
   n5 = "Transform (Modify) Bone - Bone: upperarm_r"(
    ComponentPose=@n4
    Alpha=1
    Translation=Get ReturnValueLocation__rightArm() -> (ReturnValueLocation__rightArm:Vector)
    Rotation=Get ReturnValueRotation_rightArm() -> (ReturnValueRotation_rightArm:Rotator)
    Scale="1,1,1"
   ) {
    rotationMode=BMM_Replace
    rotationSpace=BCS_WorldSpace
    scaleMode=BMM_Additive
    translationMode=BMM_Replace
    translationSpace=BCS_WorldSpace
    alphaBoolBlend {blendOption=Linear}
    alphaScaleBiasClamp {interpSpeedDecreasing=1, interpSpeedIncreasing=14}
    boneToModify {boneName=upperarm_r}
   }
   n6 = RigidBody(ComponentPose=@n5) {
    bOverrideWorldGravity=true
    overridePhysicsAsset=&/F_MED_SugarBubble_Pack_Physics
    alphaBoolBlend {blendOption=Linear}
    componentLinearAccScale=<0.3 0.3 0.3>
    componentLinearVelScale=<0.3 0.3 0.3>
    overrideWorldGravity {z=-980}
    simSpaceSettings {
     maxAngularAcceleration=500
     maxAngularVelocity=3
     maxLinearAcceleration=400
     maxLinearVelocity=250
     velocityScaleZ=0
     externalLinearDragV=<0.1 0.1>
    }
   }
   n7 = "Transform (Modify) Bone - Bone: charm_1_dyn"!(ComponentPose=@n6, bAlphaBoolEnabled:bool=@n2, Translation="0,0,0", Rotation="-16.394,-5.144,33.444", Scale="1,1,1") -> (Pose) {
    alphaInputType=Bool
    rotationMode=BMM_Additive
    scaleMode=BMM_Additive
    translationMode=BMM_Additive
    alphaBoolBlend {blendOption=Linear}
    boneToModify {boneName=charm_1_dyn}
    rotation=<rot -16.394 -5.144 33.444>
   }
   Output Pose(Result=Component To Local(ComponentPose=@n7)) {layerGroup=None}
  ]}
 EventGraph {nodes [
   Event BlueprintUpdateAnimation()
   > n1 = Sequence!(execute) -> (then_0, then_1, then_2:Exec)
   n2 = Get Dynamic Material Instance(execute=[@n9.then, @n12.then, @n1.then_0], SkeletalMeshComponent=GetPartSkeletalMeshComponent())
   n3 = PickaxeItemDefinitionHasAnyMetaTag(Tags=Make Array("[0]"=(TagName="Cosmetics.MetaTags.SugarBubble")), bRequiresEquipped=true)
   > n4 = Branch(Condition=@n3)
   > n5 = Branch(Condition=OR Boolean!(A=Get bIsSkyDiving(), B=Get bIsTacticalSprinting(), C:bool=Get bIsPlayingEmote()) -> (ReturnValue))
   > n6 = Branch(Condition=Get bIsCrouchSprinting())
   SetScalarParameterValue(execute=[@n11.then, @n6.then, @n5.else], self=@n2, ParameterName=Hide Element 01, Value=1)
   SetScalarParameterValue(execute=[@n10.then, @n11.else, @n6.else, @n4.else], self=@n2, ParameterName=Hide Element 01, Value=0)
   n7 = Branch(execute=@n1.then_1, Condition=Get bIsSkyDiving())
   > Set Rotation_RightThigh_z_inSky(Rotation_RightThigh_z_inSky:float=GetAlphaFromReferenceTransformDelta(SkeletalMeshComponent=GetOwnerSkeletalMeshComponent(), BoneName=spine_05, BaseName=spine_04, InputRange=(X=-10,Y=-13), TwistAxis="0,0,1")) -> (Output_Get:float)
   Set Rotation_RightThigh_z_inSky(execute=@n7.else, Rotation_RightThigh_z_inSky=0)
   n8 = GetSocketTransform!(self=GetOwnerSkeletalMeshComponent(), InSocketName=upperarm_r, TransformSpace=RTS_World) -> (ReturnValue_Location:Vector, ReturnValue_Rotation:Rotator, ReturnValue_Scale:Vector)
   Set ReturnValueLocation__rightArm(execute=@n1.then_2, ReturnValueLocation__rightArm:Vector=@n8) -> (Output_Get:Vector)
   > Set ReturnValueRotation_rightArm(ReturnValueRotation_rightArm:Rotator=@n8.ReturnValue_Rotation) -> (Output_Get:Rotator)
   n9 = Event OnWeaponChanged()
   n10 = Event OnBeginSkydiving()
   Event OnEndSkydiving()
   > n11 = Branch(Condition=@n3)
   n12 = Event OnCharacterCustomizationLoaded()
   TryGetPawnOwner()
  ]}
]
variables [
 ReturnValueLocation__rightArm {type=Vector}
 ReturnValueRotation_rightArm {type=Rotator}
 Rotation_RightThigh_z_inSky {type=number}
]
